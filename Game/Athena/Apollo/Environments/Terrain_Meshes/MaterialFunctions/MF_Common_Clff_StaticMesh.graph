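# Material function that builds the shared Apollo cliff surface for terrain static meshes from world-aligned cliff diffuse, spec and normal textures.
# XY and XYZ projections feed material attribute outputs, a mobile cave variant and a height output from diffuse alpha, with tintable base color.
# Epic quality layers a world-aligned rock detail texture through overlay color, modulated spec and roughness, and added normals.
# Forward and mobile paths use constant specular and roughness.
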MaterialFunction /Game/Athena/Apollo/Environments/Terrain_Meshes/MaterialFunctions/MF_Common_Clff_StaticMesh
graph {nodes [
  FunctionInput(Preview=Constant() {r=1000}) {
   bUsePreviewValueAsDefault=true
   desc="This will need to be set in many different place if we use this input again!"
   description=Deprecated
   inputName=TilingSize
   inputType=FunctionInput_Scalar
  }
  n1 = SetMaterialAttributes!(MaterialAttributes, BaseColor=NamedRerouteUsage(), Specular=NamedRerouteUsage(), Roughness=NamedRerouteUsage(), Normal=NamedRerouteUsage()) {attributeSetTypes [
    #1
    #2
    #3
    #4
   ]}
  n2 = QualitySwitch(Default=@n1, Epic=NamedRerouteUsage())
  n3 = SetMaterialAttributes(BaseColor=NamedRerouteUsage(), Specular=Constant() {r=0.3}, Roughness=Constant() {r=0.9}, Normal=NamedRerouteUsage()) {attributeSetTypes [
    #1
    #2
    #3
    #4
   ]}
  FunctionOutput(None=ShadingPathSwitch(Default=@n2, Deferred=@n2, Forward=@n3, Mobile=@n3))
  n4 = TextureObjectParameter() {
   desc=OG: T_Apollo_Cliff_D_01
   group=Cliffs
   parameterName=CommonCliff-Diffuse
   texture=/Game/Environments/Artemis/Sets/Terrain/Textures/T_Artemis_Cliff_03B_D
  }
  n5 = TextureObjectParameter() {
   group=Cliffs
   parameterName=CommonCliff-S
   samplerType=SAMPLERTYPE_Masks
   texture=/Game/Environments/Apollo/Sets/Terrain/Textures/T_Apollo_Cliff_S_01
  }
  n6 = TextureObjectParameter() {
   group=Cliffs
   parameterName=CommonCliff-N
   samplerType=SAMPLERTYPE_Normal
   texture=/Game/Environments/Artemis/Sets/Terrain/Textures/T_Artemis_Cliff_03B_N
  }
  n7 = MaterialFunctionCall!(
   Diffuse Texture Object=@n4
   Spec Texture Object=@n5
   TextureSize=Constant() {r=700}
   --------------------------
   Use seperate settings for Z projection
   Use seperate texture for Z projection
   Texture Size Z projection
   Z Plane Diffuse Texture Object
   Z Plane Spec Texture Object
   "--------------------------      "
   ProjectionTransitionContrast
   Use Texture Alpha Heightmap to Refine the Transistion=StaticBool()
   HeightLerpTransitionContrast
   PickTheDiffuseHeightChannel
   TheDiffuseHeightChannel
   "--------------------------       "
   NormalTextureObject=@n6
   World Space Normals=StaticBool() {value=true}
   "--------------------------  "
   Export Float 4 Diffuse=StaticBool() {value=true}
   Export Float 4 Spec
   "--------------------------     "
   WorldPosition=WorldPosition() {worldPositionShaderOffset=WPT_ExcludeAllShaderOffsets}.XYZ
   World Space Normal
  ) {materialFunction=/Engine/Functions/Engine_MaterialFunctions02/Texturing/WorldAlignedTextures_Complex}
  FunctionOutput(None=ComponentMask(None=@n7.XY Diffuse Texture) {a=true}) {outputName=Height, sortPriority=3}
  n8 = SetMaterialAttributes(BaseColor=NamedRerouteUsage(), Specular=Constant() {r=0.3}, Roughness=Constant() {r=0.9}, Normal=NamedRerouteUsage()) {attributeSetTypes [
    #1
    #2
    #3
    #4
   ]}
  FunctionOutput(None=@n8) {
   desc=This is used exclusively for the mobile path in Apollo_Terrain_Cave
   description=This is used exclusively for the mobile path in Apollo_Terrain_Cave
   outputName=XYZ-Mobile-Cave
   sortPriority=4
  }
  n9 = SetMaterialAttributes(BaseColor=NamedRerouteUsage(), Specular=NamedRerouteUsage(), Roughness=NamedRerouteUsage(), Normal=NamedRerouteUsage()) {attributeSetTypes [
    #1
    #2
    #3
    #4
   ]}
  n10 = NamedRerouteUsage()
  FunctionOutput(None=ShadingPathSwitch(Default=QualitySwitch(Default=@n9, Epic=@n10), Forward=@n8, Mobile=@n8)) {description=This is used exclusively for the mobile path in Apollo_Terrain_Cave, outputName=XYZ-Material, sortPriority=4}
  NamedRerouteDeclaration(None=ComponentMask(None=@n7.XYZ Spec) {b=true}) {name=XYZ-Rough, nodeColor=<rgba 1 0 0.9844 1>}
  NamedRerouteDeclaration(None=@n7.Normal XYZ Texture) {name=XYZ-Normal, nodeColor=<rgba 1 0 0.9844 1>}
  NamedRerouteDeclaration(None=ComponentMask(None=@n7.XYZ Spec) {r=true}) {name=XYZ-Spec, nodeColor=<rgba 1 0 0.9844 1>}
  n11 = FunctionInput(Preview=Constant() {r=1}) {bUsePreviewValueAsDefault=true, inputName=ColorTint, inputType=FunctionInput_Scalar}
  NamedRerouteDeclaration(None=Multiply(A=@n11, B=ComponentMask(None=@n7.XYZ Diffuse Texture) {b=true, g=true, r=true})) {name=XYZ-Diffuse, nodeColor=<rgba 1 0 0.9844 1>}
  n12 = MaterialFunctionCall(
   Diffuse Texture Object=TextureObject() {texture=/Game/Environments/Apollo/Rocks/Natural_Rocks/Textures/T_RockDetails_02D}
   TextureSize=Constant() {r=256}
   ProjectionTransitionContrast=Constant() {r=10}
   Use Texture Alpha Heightmap to Refine the Transistion=StaticBool()
   NormalTextureObject=TextureObject() {samplerType=SAMPLERTYPE_Normal, texture=/Game/Environments/Apollo/Rocks/Natural_Rocks/Textures/T_RockDetails_02_N}
   World Space Normals=StaticBool()
   Export Float 4 Diffuse=StaticBool() {value=true}
   WorldPosition=WorldPosition() {worldPositionShaderOffset=WPT_ExcludeAllShaderOffsets}.XYZ
  ) {materialFunction=&WorldAlignedTextures_Complex}
  n13 = ComponentMask(None=@n12.XYZ Diffuse Texture) {r=true}
  n14 = Constant() {r=-0.773}
  n15 = Constant() {r=1.8}
  n16 = Multiply(A=@n12.Normal XYZ Texture, B=Constant3Vector() {constant=<rgba 1 1 0 0>})
  n17 = NamedRerouteUsage()
  n18 = SetMaterialAttributes(
   BaseColor=MaterialFunctionCall(Base=NamedRerouteUsage(), Blend=ComponentMask(None=@n12.XYZ Diffuse Texture) {b=true, g=true, r=true}) {materialFunction=&Blend_Overlay}.Result
   Specular=Multiply(A=@n13, B=NamedRerouteUsage()) {desc=spec}
   Roughness=Multiply(A=LinearInterpolate(A=@n14, B=@n15, Alpha=@n13), B=NamedRerouteUsage()) {desc=rough}
   Normal=Normalize(VectorInput=Add(A=@n16, B=@n17))
  ) {attributeSetTypes [
    #1
    #2
    #3
    #4
   ]}
  NamedRerouteDeclaration(None=@n18) {name=XYZ-EPIC, nodeColor=<rgba 1 0 0.9844 1>}
  n19 = ComponentMask(None=@n12.XY Diffuse Texture) {r=true}
  n20 = Multiply(A=@n12.Normal XY Texture, B=Constant3Vector() {constant=<rgba 1 1 0 0>})
  n21 = NamedRerouteUsage()
  n22 = SetMaterialAttributes(
   BaseColor=MaterialFunctionCall(Base=NamedRerouteUsage(), Blend=ComponentMask(None=@n12.XY Diffuse Texture) {b=true, g=true, r=true}) {materialFunction=&Blend_Overlay}.Result
   Specular=Multiply(A=@n19, B=NamedRerouteUsage()) {desc=spec}
   Roughness=Multiply(A=LinearInterpolate(A=@n14, B=@n15, Alpha=@n19), B=NamedRerouteUsage()) {desc=rough}
   Normal=Normalize(VectorInput=Add(A=@n20, B=@n21))
  ) {attributeSetTypes [
    #1
    #2
    #3
    #4
   ]}
  NamedRerouteDeclaration(None=@n22) {name=XY-EPIC, nodeColor=<rgba 1 0 0.9844 1>}
  NamedRerouteDeclaration(None=@n7.Normal XY Texture) {name=XY-Normal, nodeColor=<rgba 1 0 0.9844 1>}
  NamedRerouteDeclaration(None=Multiply(A=@n11, B=ComponentMask(None=@n7.XY Diffuse Texture) {b=true, g=true, r=true})) {name=XY-Diffuse, nodeColor=<rgba 1 0 0.9844 1>}
  NamedRerouteDeclaration(None=ComponentMask(None=@n7.XY Spec) {r=true}) {name=XY-Spec, nodeColor=<rgba 1 0 0.9844 1>}
  NamedRerouteDeclaration(None=ComponentMask(None=@n7.XY Spec) {b=true}) {name=XY-Rough, nodeColor=<rgba 1 0 0.9844 1>}
  ScalarParameter() {defaultValue=700, group=Cliffs, parameterName=Cliff-Scale}
  ScalarParameter() {defaultValue=256, group=Detail Texture, parameterName=MF-CCSM-Detail-Scale}
  ScalarParameter() {defaultValue=10, group=Detail Texture, parameterName=MF-CCSM-Detail-PTC}
  ScalarParameter() {defaultValue=-0.773, group=Detail Texture, parameterName=Detail R Min, sortPriority=5}
  ScalarParameter() {defaultValue=1.8, group=Detail Texture, parameterName=Detail R Max, sortPriority=6}
 ]}
