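# Material function for Apollo terrain meshes that supplies a generic, parameterized custom material layer, grouped as BlendedCustomMaterial, as material attributes plus a height value.
# Default shading combines tinted diffuse, scaled specular from the mask's red channel, roughness remapped from its blue channel and a transformed normal map.
# Forward and mobile paths keep the tinted diffuse but use constant specular and roughness with the vertex normal.
# Height comes from scaled diffuse alpha.
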MaterialFunction /Game/Athena/Apollo/Environments/Terrain_Meshes/MaterialFunctions/MF_Common_GenericMaterial
graph {nodes [
  n1 = VectorParameter() {group=BlendedCustomMaterial, parameterName=ColorTint_CustomMaterial, sortPriority=3, defaultValue=<rgba 1 1 1 1>}
  n2 = TextureSampleParameter2D() {
   group=BlendedCustomMaterial
   parameterName=Diffuse
   samplerSource=SSM_Wrap_WorldGroupSettings
   sortPriority=51
   texture=/Game/Characters/Fortress_Character/Textures/T_All_White
  }
  n3 = Saturate(None=Multiply(A=@n1.RGB, B=@n2.RGB))
  n4 = ScalarParameter() {defaultValue=1, group=BlendedCustomMaterial, parameterName=Specular_BlendCustomMaterial, sortPriority=0}
  n5 = TextureSampleParameter2D() {
   group=BlendedCustomMaterial
   parameterName=Specular
   samplerSource=SSM_Wrap_WorldGroupSettings
   samplerType=SAMPLERTYPE_Masks
   sortPriority=52
   texture=&T_Fortnite_Default_S
  }
  n6 = ScalarParameter() {group=BlendedCustomMaterial, parameterName=Roughness_BlendCustomMaterial_low, sortPriority=1}
  n7 = ScalarParameter() {defaultValue=1, group=BlendedCustomMaterial, parameterName=Roughness_BlendCustomMaterial_High, sortPriority=2}
  n8 = TextureSampleParameter2D() {
   group=BlendedCustomMaterial
   parameterName=Normal
   samplerSource=SSM_Wrap_WorldGroupSettings
   samplerType=SAMPLERTYPE_Normal
   sortPriority=53
   texture=&FlatNormalBlue
  }
  n9 = SetMaterialAttributes!(
   MaterialAttributes
   BaseColor=@n3
   Specular=Saturate(None=Multiply(A=@n4, B=@n5.R))
   Roughness=Saturate(None=LinearInterpolate(A=@n6, B=@n7, Alpha=@n5.B))
   Normal=Transform(None=@n8.RGB)
  ) {attributeSetTypes [
    #1
    #2
    #3
    #4
   ]}
  n10 = SetMaterialAttributes(BaseColor=@n3, Specular=Constant() {r=0.3}, Roughness=Constant() {r=0.9}, Normal=VertexNormalWS()) {attributeSetTypes [
    #1
    #2
    #3
    #4
   ]}
  FunctionOutput(None=ShadingPathSwitch(Default=@n9, Forward=@n10, Mobile=@n10))
  n11 = ScalarParameter() {defaultValue=1, desc=Uses diffuse alpha channel, group=BlendedCustomMaterial, parameterName=Heightmap_BlendCustomMaterial_Mod}
  FunctionOutput(None=Multiply(A=@n11, B=@n2.A)) {outputName=Height, sortPriority=3}
 ]}
