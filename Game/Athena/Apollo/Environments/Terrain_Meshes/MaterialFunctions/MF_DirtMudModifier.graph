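# Material function that turns input gravel material attributes into mud for Apollo terrain, outputting the mud material and a colorized diffuse.
# Base color, optionally from a separate unmodified diffuse input, is tinted through MF_GravelAndMudColors and MF_BiomeColors.
# An inverted-heightmap puddle mask blends specular and roughness multipliers and flattens normals toward the vertex normal.
# Forward and Mobile shading paths override only base color, skipping the puddle effects.
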MaterialFunction /Game/Athena/Apollo/Environments/Terrain_Meshes/MaterialFunctions/MF_DirtMudModifier
graph {nodes [
  Constant() {r=-0.5462}
  Constant() {r=0.1333}
  Constant() {r=-0.1061}
  Constant() {r=1.266}
  Constant() {r=1}
  Constant() {r=1}
  FunctionOutput(None=NamedRerouteUsage()) {outputName=MudMaterial}
  n1 = Multiply(
   A=MaterialFunctionCall() {materialFunction=/Game/Athena/Apollo/Environments/Landscape/Materials/Biome/CommonMaterials/MF_GravelAndMudColors}.MudColorMod
   B=NamedRerouteUsage()
  )
  n2 = FunctionInput(Preview=StaticBool()) {
   bUsePreviewValueAsDefault=true
   description="Default=False. When set to true, this is designed to hide the \"BiomeColors-UseDefault\" bool param in material instances that don't need it - for UEFN"
   inputName=BypassBiomeColors
   inputType=FunctionInput_StaticBool
   sortPriority=100
  }
  FunctionOutput(None=NamedRerouteDeclaration(None=MaterialFunctionCall!(Grass, Gravel, Mud=@n1, Sand, UseLandscapeCoords, BypassBiomeColors=@n2) {materialFunction=/Game/Athena/Apollo/Environments/Landscape/Materials/Biome/CommonMaterials/MF_BiomeColors}.Mud) {name=Mud-Basecolor, nodeColor=<rgba 1 0 0.9844 1>}) {outputName=ColorizedGravelDiffuse, sortPriority=10}
  n3 = GetMaterialAttributes(None=FunctionInput() {inputName=GravelMaterialInput, inputType=FunctionInput_MaterialAttributes}) {attributeGetTypes [
    #1
    #2
    #3
    #4
   ]}
  NamedRerouteDeclaration(None=@n3.Specular) {name=Input-Spec, nodeColor=<rgba 1 0 0.9844 1>}
  NamedRerouteDeclaration(None=@n3.Roughness) {name=Input-Rough, nodeColor=<rgba 1 0 0.9844 1>}
  n4 = FunctionInput() {bUsePreviewValueAsDefault=true, inputName=UnmodifiedGravelDiffuse, sortPriority=20, previewValue=<1 1 1 1>}
  n5 = FunctionInput(Preview=StaticBool()) {
   bUsePreviewValueAsDefault=true
   description="Default=False"
   inputName=UseUnmodifiedGravelDiffuse
   inputType=FunctionInput_StaticBool
   sortPriority=10
  }
  NamedRerouteDeclaration(None=StaticSwitch(True=@n4, False=@n3.BaseColor, Value=@n5)) {name=Input-Basecolor, nodeColor=<rgba 1 0 0.9844 1>}
  NamedRerouteDeclaration(None=@n3.Normal) {name=Input-Normal, nodeColor=<rgba 1 0 0.9844 1>}
  NamedRerouteDeclaration(None=@n3.MaterialAttributes) {name=Input-Material, nodeColor=<rgba 1 0 0.9844 1>}
  n6 = OneMinus(None=FunctionInput() {inputName=Heightmap, inputType=FunctionInput_Scalar, sortPriority=3})
  n7 = ScalarParameter() {group=Mud, parameterName=MudPuddleHeight}
  n8 = Saturate(None=Divide(A=Add(A=@n6, B=@n7), B=ScalarParameter() {defaultValue=1, group=Mud, parameterName=MudPuddleBlend}))
  n9 = LinearInterpolate(
   A=ScalarParameter() {defaultValue=0.6, group=Mud, parameterName=Specular_MudHigh}
   B=ScalarParameter() {defaultValue=0.6, group=Mud, parameterName=Specular_MudLow}
   Alpha=@n8
  )
  n10 = LinearInterpolate(
   A=ScalarParameter() {defaultValue=0.8, group=Mud, parameterName=Roughness_MudHigh}
   B=ScalarParameter() {defaultValue=0.3, group=Mud, parameterName=Roughness_MudLow}
   Alpha=@n8
  )
  n11 = SetMaterialAttributes!(
   MaterialAttributes=NamedRerouteUsage()
   BaseColor=NamedRerouteUsage()
   Specular=Multiply(A=NamedRerouteUsage(), B=Saturate(None=@n9))
   Roughness=Multiply(A=NamedRerouteUsage(), B=Saturate(None=@n10))
   Normal=LinearInterpolate(A=NamedRerouteUsage(), B=VertexNormalWS(), Alpha=@n8)
  ) {attributeSetTypes [
    #1
    #2
    #3
    #4
   ]}
  n12 = SetMaterialAttributes!(MaterialAttributes=NamedRerouteUsage(), BaseColor=NamedRerouteUsage()) {attributeSetTypes [#1]}
  NamedRerouteDeclaration(None=ShadingPathSwitch(Default=@n11, Forward=@n12, Mobile=@n12)) {name=MudMaterial, nodeColor=<rgba 1 0 0.9844 1>}
  n13 = Multiply(A=Constant3Vector() {constant=<rgba 0.5 0.4596 0.375 1>}, B=NamedRerouteUsage())
  Reroute(None=@n13)
  Reroute(None=@n13)
  n14 = StaticSwitchParameter(True=BlendMaterialAttributes(A=NamedRerouteUsage(), B=NamedRerouteUsage(), Alpha=VertexColor().A), False=NamedRerouteUsage()) {
   desc="UseCliffs should be disabled since this uses vertex alpha channel as a mask.\r\nIf this equals true, then UseCliffs should equal false.\r\nMud uses vertex alpha for the mask.\r\nCliffs use vertex alpha for the mask."
   group=Layers
   parameterName=UseMud
   sortPriority=3
  }
  Reroute(None=@n14)
 ]}
