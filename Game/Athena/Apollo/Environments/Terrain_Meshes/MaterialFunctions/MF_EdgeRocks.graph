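# Material function that builds world-aligned cliff rock layers for edge rocks on Apollo terrain.
# WorldAlignedTextures_Complex projects cliff diffuse, spec and normal textures, by default shearing world Z by (X+Y)/ShiftDistance to slant the pattern.
# It outputs tinted material attributes for XY walls and Z ceiling, plus diffuse-alpha heights for each.
# Only wall normals are flattened; the ColorTint_RockCliff parameter is unconnected.
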MaterialFunction /Game/Athena/Apollo/Environments/Terrain_Meshes/MaterialFunctions/MF_EdgeRocks
graph {nodes [
  n1 = FunctionInput(Preview=Constant() {r=1}) {bUsePreviewValueAsDefault=true, inputName=ColorTint, inputType=FunctionInput_Scalar}
  n2 = WorldPosition() {worldPositionShaderOffset=WPT_ExcludeAllShaderOffsets}
  n3 = WorldPosition() {worldPositionShaderOffset=WPT_ExcludeAllShaderOffsets}
  n4 = ComponentMask(None=@n3.XYZ) {r=true}
  n5 = ComponentMask(None=@n3.XYZ) {g=true}
  n6 = AppendVector(A=Constant2Vector(), B=Divide(A=Add(A=@n4, B=@n5), B=ScalarParameter() {defaultValue=3.2978, parameterName=ShiftDistance}))
  n7 = StaticBool() {value=true}
  n8 = MaterialFunctionCall!(
   Diffuse Texture Object=TextureObject() {texture=/Game/Environments/Apollo/Sets/Terrain/Textures/T_Apollo_Cliff_D_02}
   Spec Texture Object=TextureObject() {samplerType=SAMPLERTYPE_Masks, texture=/Game/Environments/Apollo/Sets/Terrain/Textures/T_Apollo_Cliff_S_02}
   TextureSize=FunctionInput(Preview=Constant() {r=1000}) {bUsePreviewValueAsDefault=true, inputName=TilingSize, inputType=FunctionInput_Scalar}
   --------------------------
   Use seperate settings for Z projection
   Use seperate texture for Z projection
   Texture Size Z projection
   Z Plane Diffuse Texture Object
   Z Plane Spec Texture Object
   "--------------------------      "
   ProjectionTransitionContrast
   Use Texture Alpha Heightmap to Refine the Transistion=StaticBool()
   HeightLerpTransitionContrast
   PickTheDiffuseHeightChannel
   TheDiffuseHeightChannel
   "--------------------------       "
   NormalTextureObject=TextureObject() {samplerType=SAMPLERTYPE_Normal, texture=/Game/Environments/Apollo/Sets/Terrain/Textures/T_Apollo_Cliff_N_02}
   World Space Normals=StaticBool() {value=true}
   "--------------------------  "
   Export Float 4 Diffuse=StaticBool() {value=true}
   Export Float 4 Spec
   "--------------------------     "
   WorldPosition=StaticSwitch(
    True=Add(A=@n2.XYZ, B=@n6)
    False=@n2.XYZ
    Value=FunctionInput(Preview=@n7) {bUsePreviewValueAsDefault=true, inputName=Use_Shift_Worldposition, inputType=FunctionInput_StaticBool, sortPriority=232}
   )
   World Space Normal
  ) {materialFunction=/Engine/Functions/Engine_MaterialFunctions02/Texturing/WorldAlignedTextures_Complex}
  n9 = SetMaterialAttributes!(
   MaterialAttributes
   BaseColor=Multiply(A=@n1, B=ComponentMask(None=@n8.XY Diffuse Texture) {b=true, g=true, r=true})
   Specular=ComponentMask(None=@n8.XY Spec) {r=true}
   Roughness=ComponentMask(None=@n8.XY Spec) {b=true}
   Normal=MaterialFunctionCall(Normal=@n8.Normal XY Texture, Flatness=ScalarParameter() {defaultValue=0.15, parameterName=FlattenNormal_EdgeRocks}) {materialFunction=&FlattenNormal}.Result
  ) {attributeSetTypes [
    #1
    #2
    #3
    #4
   ]}
  FunctionOutput(None=@n9) {outputName=XY Walls}
  FunctionOutput(None=ComponentMask(None=@n8.XY Diffuse Texture) {a=true}) {outputName=XY Walls Height, sortPriority=3}
  n10 = SetMaterialAttributes(
   BaseColor=Multiply(A=@n1, B=ComponentMask(None=@n8.Z Diffuse Texture) {b=true, g=true, r=true})
   Specular=ComponentMask(None=@n8.Z Spec) {r=true}
   Roughness=ComponentMask(None=@n8.Z Spec) {b=true}
   Normal=@n8.Normal Z Texture
  ) {attributeSetTypes [
    #1
    #2
    #3
    #4
   ]}
  FunctionOutput(None=@n10) {outputName="Z Ceiling ", sortPriority=8}
  FunctionOutput(None=ComponentMask(None=@n8.Z Diffuse Texture) {a=true}) {outputName=Z Ceiling Height, sortPriority=9}
  VectorParameter() {parameterName=ColorTint_RockCliff, defaultValue=<rgba 0.8698 0.8589 0.6537 1>}
 ]}
