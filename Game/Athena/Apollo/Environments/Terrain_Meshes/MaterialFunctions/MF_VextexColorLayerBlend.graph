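# Material function that layers grass, gravel, mud or sand, asphalt, and cliff or custom materials on Apollo terrain meshes by vertex color.
# Blue drives gravel, green mud/sand within gravel, red asphalt, and alpha cliffs or the custom material.
# Static switches select layers; height lerps shape transitions, with darkened edges at asphalt and cliff seams.
# Forward and mobile paths use vertex normals beneath the cliff/custom blend.
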MaterialFunction /Game/Athena/Apollo/Environments/Terrain_Meshes/MaterialFunctions/MF_VextexColorLayerBlend
graph {nodes [
  Constant3Vector() {constant=<rgba 1 0 0.6086 0>}
  Constant3Vector() {constant=<rgba 0.65 0.608 0.52 1>}
  FunctionInput() {desc=Not used, inputName=CustomMaterial_Height, inputType=FunctionInput_Scalar, sortPriority=17}
  n1 = GetMaterialAttributes(None=NamedRerouteUsage()) {attributeGetTypes [#1]}
  n2 = MaterialFunctionCall(
   Transition Phase=VertexColor() {desc=mask cliff to grass transition}.A
   Height Texture=OneMinus(None=NamedRerouteUsage())
   Contrast=ScalarParameter() {group=Cliffs, parameterName=CustomMaterialLerpContrast}
  ) {materialFunction=&HeightLerp}
  n3 = Multiply(A=@n1.BaseColor, B=@n2.Lerp Alpha No Contrast)
  n4 = NamedRerouteUsage()
  n5 = GetMaterialAttributes(None=NamedRerouteUsage()) {attributeGetTypes [#1]}
  n6 = MaterialFunctionCall(
   Transition Phase=VertexColor() {desc=mask cliff to grass transition}.A
   Height Texture=OneMinus(None=NamedRerouteUsage())
   Contrast=ScalarParameter() {group=Cliffs, parameterName=CliffLerpContrast}
  ) {materialFunction=&HeightLerp}
  n7 = SetMaterialAttributes!(MaterialAttributes=@n5.MaterialAttributes, BaseColor=Multiply(A=@n5.BaseColor, B=@n6.Lerp Alpha No Contrast)) {desc=Grass shadow, attributeSetTypes [#1]}
  n8 = StaticSwitchParameter(
   True=BlendMaterialAttributes(
    A=NamedRerouteUsage()
    B=SetMaterialAttributes(MaterialAttributes=@n1.MaterialAttributes, BaseColor=@n3) {desc=Grass shadow, attributeSetTypes [#1]}
    Alpha=@n2.Alpha
   )
   False=StaticSwitchParameter(True=BlendMaterialAttributes(A=@n4, B=@n7, Alpha=@n6.Alpha), False=NamedRerouteUsage()) {desc=Uses Alpha vertex channel., group=Layers, parameterName=UseCliffs, sortPriority=7}
  ) {desc=Uses Alpha vertex channel. True replaces the Cliffs material., group=Layers, parameterName=UseCustomMaterialBlend, sortPriority=8}
  FunctionOutput(None=@n8)
  n9 = MaterialFunctionCall!(
   OverrideEdgeColor=MaterialFunctionCall() {materialFunction=/Game/Athena/Apollo/Environments/Landscape/Materials/Biome/CommonMaterials/MF_GravelAndMudColors}.SM-GravelEdgeColorMod
   EdgeMask=NamedRerouteUsage()
   BypassBiomeColors
  ) {
   desc="Gravel used to be referred to as \"Dirt\" in the legacy terrain material"
   materialFunction=/Game/Athena/Apollo/Environments/Landscape/Materials/Biome/CommonMaterials/Mobile/MF_Mobile_Common_Gravel
  }
  NamedRerouteDeclaration(None=@n9.Height) {name=Gravel-Height, nodeColor=<rgba 1 0 0.9844 1>}
  n10 = VertexColor() {desc=mask cliff to grass transition}
  n11 = StaticSwitchParameter(True=Saturate(None=Add(A=@n10.R, B=@n10.R)), False=@n10.B) {defaultValue=true, desc=True will also load the Gravel material, group=Layers, parameterName=UseAsphalt, sortPriority=6}
  n12 = NamedRerouteUsage()
  NamedRerouteDeclaration(None=MaterialFunctionCall(Transition Phase=@n11, Height Texture=OneMinus(None=@n12)) {materialFunction=&HeightLerp}.Lerp Alpha No Contrast) {name=GravelEdgeColor-Mask, nodeColor=<rgba 1 0 0.9844 1>}
  n13 = StaticSwitchParameter(True=BlendMaterialAttributes(A=NamedRerouteUsage(), B=NamedRerouteUsage(), Alpha=NamedRerouteUsage()), False=NamedRerouteUsage()) {
   defaultValue=true
   desc="If True, then Grass will be shown where the vertex color is black."
   group=Layers
   parameterName=UseGrass
   sortPriority=5
  }
  n14 = StaticSwitchParameter(True=@n13, False=NamedRerouteUsage()) {
   defaultValue=true
   desc="Uses Blue vertex channel. False = Grass, but no Gravel, Mud, Sand or Asphalt edge color"
   group=Layers
   parameterName=UseGravel
   sortPriority=2
  }
  n15 = StaticSwitchParameter(True=NamedRerouteUsage(), False=NamedRerouteDeclaration(None=@n14) {name=Grass-Gravel-Mud/Sand, nodeColor=<rgba 1 0 0.9844 1>}) {
   defaultValue=true
   desc=Uses Red vertex channel. True will also load the Gravel material
   group=Layers
   parameterName=UseAsphalt
   sortPriority=6
  }
  n16 = VertexNormalWS()
  n17 = SetMaterialAttributes!(MaterialAttributes=@n15, Normal=ShadingPathSwitch(Default=@n15, Forward=@n16, Mobile=@n16)) {
   desc="This is assuming there are no normal textures on mobile other than the Cliffs && Custom Material"
   attributeSetTypes [#2]
  }
  NamedRerouteDeclaration(None=@n17) {name=Possible-GrassGravelMud/SandAsphalt, nodeColor=<rgba 1 0 0.9844 1>}
  NamedRerouteDeclaration(None=@n9.ForStaticMesh) {name=Gravel-Material, nodeColor=<rgba 1 0 0.9844 1>}
  NamedRerouteDeclaration(None=MaterialFunctionCall!(CustomSeaLevel, BypassBiomeColors, UseOceanLevel) {materialFunction=/Game/Athena/Apollo/Environments/Landscape/Materials/Biome/CommonMaterials/Mobile/MF_Mobile_Common_Sand}.ForStaticmesh) {name=Sand-Material, nodeColor=<rgba 1 0 0.9844 1>}
  n18 = MaterialFunctionCall!(
   GravelMaterialInput=NamedRerouteUsage()
   Heightmap=NamedRerouteUsage()
   UseUnmodifiedGravelDiffuse
   UnmodifiedGravelDiffuse
   BypassBiomeColors
  ) {
   desc="Creates the Mud layer using the Gravel layer\r\nusing a static switch param.\r\nUses vertex alpha channel"
   materialFunction=&/MF_DirtMudModifier
  }
  NamedRerouteDeclaration(None=@n18.MudMaterial) {name=Gravel-Mud-Blend, nodeColor=<rgba 1 0 0.9844 1>}
  n19 = Constant() {r=-0.55}
  n20 = Constant() {r=2}
  n21 = VertexColor()
  n22 = Constant() {r=21.251}
  n23 = MaterialFunctionCall(Transition Phase=Subtract(A=@n21.R, B=Power(Base=@n21.B, Exp=@n22)), Height Texture=NamedRerouteUsage()) {materialFunction=&HeightLerp}
  NamedRerouteDeclaration(None=Saturate(None=LinearInterpolate(A=@n19, B=@n20, Alpha=@n23.Lerp Alpha No Contrast))) {name=AsphaltEdgeMaskA, nodeColor=<rgba 1 0 0.9844 1>}
  n24 = FunctionInput() {desc=The input normals need to be in worldspace, inputName=CustomMaterial, inputType=FunctionInput_MaterialAttributes, sortPriority=16}
  NamedRerouteDeclaration(None=@n24) {name=CustomMaterial, nodeColor=<rgba 1 0 0.9844 1>}
  n25 = MaterialFunctionCall!(
   ColorTint=MaterialFunctionCall!(DarknessAmount=Constant() {r=0.2}) {materialFunction=/Game/Athena/Apollo/Environments/Landscape/Materials/Biome/CommonMaterials/MF_HeightBasedDarkenCliff}.Result
   TilingSize
  ) {
   desc="Cliff height doesn't work well with height lerp on the terrain\r\nmeshes - it stretches out onto the flat vertical surfaces."
   materialFunction=&/MF_Common_Clff_StaticMesh
  }
  NamedRerouteDeclaration(None=StaticSwitchParameter(True=@n25.XYZ-Material, False=@n25.XYZ-Material) {desc="Only use this for blending to the cave meshes that have a \"ceiling\"! Adds 40 ins.", parameterName=Use-XYZ-CliffBlendToCave}) {name=CliffsMaterial, nodeColor=<rgba 1 0 0.9844 1>}
  n26 = Constant() {r=-0.33}
  n27 = Constant() {r=1.25}
  n28 = MaterialFunctionCall(Transition Phase=VertexColor().R, Height Texture=NamedRerouteUsage()) {materialFunction=&HeightLerp}
  NamedRerouteDeclaration(None=Saturate(None=LinearInterpolate(A=@n26, B=@n27, Alpha=@n28.Lerp Alpha No Contrast))) {name=AsphaltEdgeMaskB, nodeColor=<rgba 1 0 0.9844 1>}
  NamedRerouteDeclaration(None=MaterialFunctionCall(
    GravelMaterialInput=NamedRerouteUsage()
    Heightmap=NamedRerouteUsage()
    UseUnmodifiedGravelDiffuse=StaticBool() {value=true}
    UnmodifiedGravelDiffuse=@n9.RGB
   ) {materialFunction=&MF_DirtMudModifier}.MudMaterial) {name=MudMaterial, nodeColor=<rgba 1 0 0.9844 1>}
  n29 = MaterialFunctionCall!(CombinedLandscapeLayersMask=NamedRerouteUsage(), BypassUseRVT, BypassBiomeColors) {materialFunction=/Game/Athena/Apollo/Environments/Landscape/Materials/Biome/CommonMaterials/Mobile/MF_Mobile_Common_Grass}
  NamedRerouteDeclaration(None=@n29.Height) {name=Grass-Height, nodeColor=<rgba 1 0 0.9844 1>}
  n30 = GetMaterialAttributes(None=NamedRerouteUsage()) {attributeGetTypes [#1]}
  n31 = NamedRerouteUsage()
  n32 = LinearInterpolate(
   A=@n30.BaseColor
   B=Multiply(A=GetMaterialAttributes(None=@n31) {attributeGetTypes [#1]}.BaseColor, B=Constant() {r=0.3})
   Alpha=NamedRerouteUsage()
  )
  n33 = GetMaterialAttributes(None=NamedRerouteUsage()) {attributeGetTypes [#2, #3]}
  n34 = NamedRerouteUsage()
  n35 = Constant() {r=0.4}
  n36 = SetMaterialAttributes!(
   MaterialAttributes=@n33.MaterialAttributes
   Normal=LinearInterpolate(A=NamedRerouteUsage(), B=@n33.Normal, Alpha=@n34)
   Specular=LinearInterpolate(A=Multiply(A=@n33.Specular, B=@n35), B=@n33.Specular, Alpha=@n34)
  ) {attributeSetTypes [#2, #3]}
  n37 = NamedRerouteUsage()
  NamedRerouteDeclaration(None=BlendMaterialAttributes(
    A=SetMaterialAttributes!(MaterialAttributes=@n30.MaterialAttributes, BaseColor=@n32) {attributeSetTypes [#1]}
    B=ShadingPathSwitch(Default=@n36, Forward=@n37, Mobile=@n37)
    Alpha=NamedRerouteUsage()
   )) {name=Grass-Gravel-Mud/Sand-Asphalt, nodeColor=<rgba 1 0 0.9844 1>}
  NamedRerouteDeclaration(None=@n29.ForStaticMesh) {name=Grass-Material, nodeColor=<rgba 1 0 0.9844 1>}
  n38 = NamedRerouteUsage()
  n39 = NamedRerouteUsage()
  n40 = VertexColor()
  n41 = NamedRerouteUsage()
  n42 = StaticSwitchParameter(
   True=BlendMaterialAttributes(
    A=NamedRerouteUsage()
    B=StaticSwitchParameter(True=@n38, False=@n39) {desc=Uses Green vertex channel., group=Layers, parameterName="Sand=True-Mud=False", sortPriority=4}
    Alpha=MaterialFunctionCall(Transition Phase=@n40.G, Height Texture=@n41) {materialFunction=&HeightLerp}.Lerp Alpha No Contrast
   )
   False=NamedRerouteUsage()
  ) {desc="You can only paint Mud or Sand where Gravel has been painted!", group=Layers, parameterName=UseMudOrSand, sortPriority=3}
  NamedRerouteDeclaration(None=@n42) {name=Gravel-Mud/Sand, nodeColor=<rgba 1 0 0.9844 1>}
  n43 = NamedRerouteUsage()
  n44 = Constant() {r=1.968}
  n45 = NamedRerouteUsage()
  n46 = Constant() {r=1.97}
  NamedRerouteDeclaration(None=MaterialFunctionCall!(
    Diffuse 1
    Diffuse 2
    Transistion Phase=OneMinus(None=@n11)
    Height Texture 1=Multiply(A=@n43, B=@n44)
    Height Texture 2=Multiply(A=@n45, B=@n46)
    Contrast
   ) {materialFunction=/Engine/Functions/Engine_MaterialFunctions02/Texturing/HeightLerpWithTwoHeightMaps}.Lerp Alpha No Contrast) {name=Grass-BlendMask, nodeColor=<rgba 1 0 0.9844 1>}
  n47 = MaterialFunctionCall!(TilingSize) {materialFunction=/Game/Athena/Apollo/Environments/Landscape/Materials/MaterialFunctions/MF_ApolloAsphalt}
  NamedRerouteDeclaration(None=@n47.ForStaticmesh) {name=Asphalt-Material, nodeColor=<rgba 1 0 0.9844 1>}
  n48 = Constant() {r=4}
  NamedRerouteDeclaration(None=Saturate(None=Multiply(A=@n11, B=@n48))) {name=GrassEdgeColor-Mask, nodeColor=<rgba 1 0 0.9844 1>}
  NamedRerouteDeclaration(None=@n47.CrackNormal) {name=Asphalt-Normals, nodeColor=<rgba 1 0 0.9844 1>}
  NamedRerouteDeclaration(None=@n47.CrackHeight) {name=Asphalt-Height, nodeColor=<rgba 1 0 0.9844 1>}
  n49 = NamedRerouteUsage()
  n50 = NamedRerouteUsage()
  n51 = MaterialFunctionCall(Transition Phase=VertexColor().G, Height Texture=NamedRerouteUsage()) {materialFunction=&HeightLerp}
  Reroute(None=StaticSwitchParameter(True=BlendMaterialAttributes(A=@n49, B=@n50, Alpha=@n51.Lerp Alpha No Contrast), False=@n49) {group=Layers, parameterName=UseSand, sortPriority=4})
  StaticSwitchParameter(
   True=BlendMaterialAttributes(A=NamedRerouteUsage(), B=NamedRerouteUsage(), Alpha=VertexColor() {desc=mask cliff to grass transition}.A)
   False=@n8
  ) {desc=Useful for simple cases like a sidewalk with a grass plane, group=Layers, parameterName=UseOnlyGrassAndCustomMaterial, sortPriority=1}
 ]}
